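# Niagara effect for the PitGlass backpack cosmetic: a glowing blue sphere with flickering lightning arcs, each lit by mesh and light renderers.
# The Sphere emitter burst-spawns a persistent particle whose color pulses via a waveform.
# SplineThickenMeshes spawns short-lived arcs at random, varying mesh index and mirrored scale, with random camera-aligned orientation.
# Curves drive material parameters, derived glow-layer variables and a flashing light offset along each arc's axis.
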NiagaraSystem /BRCosmetics/Accessories/FORT_Backpacks/Backpack_M_MED_PitGlass/FX/NS_Backpack_PitGlass
dependencies {
 usedDataInterfaces [&NiagaraDataInterfaceArrayFloat3, &NiagaraDataInterfaceCamera, &NiagaraDataInterfaceCurve]
 usedDynamicInputs []
 usedModules [&SystemState]
}
emitters {
 Sphere {
  emitterUpdateScript [
   &EmitterState(
    Inactive Response=::NewEnumerator1
    Life Cycle Mode=::NewEnumerator1
    Loop Behavior=::NewEnumerator1
    Loop Duration Mode=::NewEnumerator1
    UseLoopDelay=false
   )
   &SpawnBurst_Instantaneous(Use Spawn Probability=false, UseLoopCountLimit=false)
  ]
  particleSpawnScript [&InitializeParticle(
    Color=<rgba 0 3.3 9 0.05>
    Lifetime=5
    Mesh Scale Mode=::NewEnumerator3
    Position Offset=<0 0.1 -0.2>
    UsePositionOffset=true
    Write Material Random=true
   )]
  particleUpdateScript [
   &ParticleState(Kill Particles When Lifetime Has Elapsed=false)
   &ScaleColor(
    Scale RGB=&VectorFromFloat(Value=&Waveform(
      Waveform Count=::NewEnumerator1
      "Waveform [1]"=::NewEnumerator0
      "[1] Frequency"=0.618
      "[1] Amplitude Min/Max"=<0.25 0.5>
      "[2] Average"=false
      "[2] Input Position"=0
      "[2] Phase"=$Emitter.Age
      "[2] Phase Scale"=1
      "[2] Compound Frequency"=2
      "[2] Frequency"=1
      "[2] Amplitude Min/Max"=<0.25 0.5>
      "[2] Scale"=1
      "[2] Bias"=0
      "[2] Exponent"=1
      "[2] Pulse Duration"=0.1
      "[2] Clip Triangle Min/Max"=<0 0>
      "[2] Random Seed"=0
      "Waveform [3] Blend"=::NewEnumerator5
     ))
    ScaleA=false
    ScaleRGB=true
   )
  ]
  rendererClasses [&NiagaraMeshRendererProperties, &NiagaraLightRendererProperties]
  renderers [
   {rendererClass=&NiagaraMeshRendererProperties, rendererIndex=0}
   {rendererClass=&NiagaraMeshRendererProperties, rendererIndex=1}
   {rendererClass=&NiagaraMeshRendererProperties, rendererIndex=2}
   {rendererClass=&NiagaraLightRendererProperties, rendererIndex=3}
  ]
 }
 SplineThickenMeshes {
  emitterUpdateScript [
   &EmitterState(
    Inactive Response=::NewEnumerator1
    Life Cycle Mode=::NewEnumerator1
    Loop Duration=&RandomRangeFloat(Minimum=0.133, Maximum=1.333)
    Recalculate Duration Each Loop=true
    UseLoopDelay=false
   )
   &SpawnRate(Spawn Probability=0.333, SpawnRate=4, Use Spawn Probability=true)
   &SpawnBurst_Instantaneous(Spawn Probability=0.5, Use Spawn Probability=true, UseLoopCountLimit=false)
  ]
  particleSpawnScript [
   /SetVariables() {enabled=false, isSetParameters=true}
   &InitializeParticle(
    Color=<rgba 0 0.75 1 1>
    Lifetime Max=0.733
    Lifetime Min=0.466
    Lifetime Mode=::NewEnumerator1
    Mesh Scale=&SelectVectorFromArray(
     Array Sampling Mode=::NewEnumerator0
     Direct Array Index=?
     Vector Selection Array=DataInterface() {propertyValues {FloatData [<1 -1 1>, <1 1 -1>, <1 -1 -1>, <1 1 1>]}}
     Random Seed=0
    )
    Mesh Scale Mode=::NewEnumerator4
    Position Offset=<0 0.1 -0.2>
    Sprite Size Mode=::NewEnumerator3
    UsePositionOffset=true
    Write Material Random=true
   )
   /SetVariables(Particles.MeshIndex=&ModuloInt(A=&ReturnParticleIDAttributes(), B=6)) {isSetParameters=true}
   /SetVariables(Particles.MeshIndex=5) {enabled=false, isSetParameters=true}
  ]
  particleUpdateScript [
   &ParticleState()
   &UpdateMeshOrientation(
    Facing Direction=&RandomRangeVector(Minimum=<0 -1 -1>, Maximum=<1 1 1>)
    Facing Interpolation=true
    Facing Interpolation Rate=0.5
    Orientation Method=::NewEnumerator1
    Override Quat=false
    Side Direction:Vector3f=&ReturnCameraVector(Camera Vectors=::NewEnumerator1, Coordinate Space=::Simulation)
    Stabilization Mode:ENiagara_OrientVectorCount=::NewEnumerator2
    TransformFacing=true
    TransformSide=true
   )
   &DynamicMaterialParameters(
    Index 0 Param 1=&FloatFromCurve(FloatCurve=DataInterface() {propertyValues {Curve {keys [
         {interpMode=RCIM_Cubic, leaveTangent=4.2199, tangentMode=RCTM_Break, time=0, value=0}
         <at 0.25 0.5 RCIM_Constant>
         {interpMode=RCIM_Cubic, leaveTangent=2.2211, tangentMode=RCTM_Break, time=0.5, value=0.666}
         {arriveTangent=0.0283, interpMode=RCIM_Cubic, tangentMode=RCTM_Break, time=1, value=1}
        ]}}})
    Index 0 Param 2=&FloatFromCurve(FloatCurve=DataInterface() {propertyValues {Curve {keys [<at 0 0.5>, <at 0.2 1>, <at 0.4 1>, {arriveTangent=-4.0866, time=1, value=0}]}}})
    Index 0 Param 3=&FloatFromCurve(FloatCurve=DataInterface() {propertyValues {Curve {keys [<at 0 1 RCIM_Constant>,<at 0.4 1>,<at 0.5 0 RCIM_Constant>,<at 1 0>]}}})
    Index 0 Param 4=&FloatFromCurve(
     FloatCurve=DataInterface() {propertyValues {Curve {keys [<at 0 0 RCIM_Constant>,<at 0.25 1 RCIM_Constant>,<at 0.333 0.166 RCIM_Constant>,<at 0.4 1>]}}}
     Scale Curve=33
    )
    Index 1 Param 1=&FloatFromCurve(FloatCurve=DataInterface() {propertyValues {Curve {keys [<at 0 0.5 RCIM_Constant>, <at 0.2 0.2 RCIM_Constant>, <at 0.4 0.75>, {tangentMode=RCTM_User, time=1, value=2}]}}})
    Index 1 Param 2=&FloatFromCurve(
     FloatCurve=DataInterface() {propertyValues {Curve {keys [
         {interpMode=RCIM_Cubic, leaveTangent=25.2051, tangentMode=RCTM_Break, time=0, value=0}
         <at 0.2 2 RCIM_Constant>
         <at 0.266 1>
         <at 0.333 1>
         <at 1 0>
        ]}}}
     Scale Curve=0.5
    )
    Index 1 Param 3=0.5
    Module.Index0Param0Write=true
    Module.Index0Param1Write=true
    Module.Index0Param2Write=true
    Module.Index0Param3Write=true
    Module.Index1Param0Write=true
    Module.Index1Param1Write=true
    Module.Index1Param2Write=true
    Module.Index1Param3Write=false
    Write Parameter Index 1=true
   )
   /SetVariables(
    Particles.LightColor=&MultiplyLinearColor(
     Linear Color=<rgba 0 2.7 9 0.005>
     Scale Factor=&MakeVector4_FromFloat(Float=&FloatFromCurve(FloatCurve=DataInterface() {propertyValues {Curve {keys [
           {arriveTangent=9.0725, interpMode=RCIM_Cubic, leaveTangent=9.0725, tangentMode=RCTM_User, time=0, value=0}
           {arriveTangent=-2.5249, interpMode=RCIM_Cubic, leaveTangent=-2.5249, tangentMode=RCTM_User, time=0.1408, value=0.9847}
           {arriveTangent=-2.2063, interpMode=RCIM_Cubic, leaveTangent=-2.2063, tangentMode=RCTM_User, time=0.5, value=0}
          ]}}}))
    )
    Particles.LightPosition:NiagaraPosition=&AddVectorToPosition(
     Position=$Particles.Position
     Vector=&Multiply_VectorByFloat(
      Vector=&ReturnMeshOrientationAxis(Mesh Orientation Axis=::XAxis, Mesh Orientation Quat=$Particles.MeshOrientation)
      Float=9
     )
    )
    Particles.LightRadius=&FloatFromCurve(FloatCurve=DataInterface() {propertyValues {Curve {keys [<at 0 1 RCIM_Constant>,<at 0.5 0>]}}}, Scale Curve=100)
    Particles.LightningGlowColor:LinearColor=&MultiplyLinearColor(
     Linear Color=$Particles.Color
     Scale Factor=&MakeFloat4FromVectorAndFloat("Vector (RGB)"=&VectorFromFloat(Value=0.5), "Float (Alpha)"=1)
    )
    Particles.LightningGlowDynamic0:Vector4f=&MakeVector4(
     X=&MakeFloatFromVector4(Vector 4=$Particles.DynamicMaterialParameter, Channel=::NewEnumerator0)
     Y=&MakeFloatFromVector4(Vector 4=$Particles.DynamicMaterialParameter, Channel=::NewEnumerator1)
     Z=0
     W=0
    )
    Particles.LightningGlowDynamic1:Vector4f=&MakeVector4(
     X=&MakeFloatFromVector4(Vector 4=$Particles.DynamicMaterialParameter1, Channel=::NewEnumerator0)
     Y=&Multiply_Float(A=&MakeFloatFromVector4(Vector 4=$Particles.DynamicMaterialParameter1, Channel=::NewEnumerator1), B=2)
     Z=&MakeFloatFromVector4(Vector 4=$Particles.DynamicMaterialParameter1, Channel=::NewEnumerator2)
     W=0
    )
    Particles.LightningGlowSoftColor:LinearColor=&MultiplyLinearColor(
     Linear Color=$Particles.Color
     Scale Factor=&MakeFloat4FromVectorAndFloat("Vector (RGB)"=&VectorFromFloat(Value=0.166), "Float (Alpha)"=1)
    )
    Particles.LightningGlowSoftDynamic1:Vector4f=&MakeVector4(
     X=&MakeFloatFromVector4(Vector 4=$Particles.DynamicMaterialParameter, Channel=::NewEnumerator0)
     Y=&Multiply_Float(A=&MakeFloatFromVector4(Vector 4=$Particles.DynamicMaterialParameter1, Channel=::NewEnumerator1), B=6)
     Z=&MakeFloatFromVector4(Vector 4=$Particles.DynamicMaterialParameter1, Channel=::NewEnumerator2)
     W=0
    )
    Particles.LightningGlowSoftTopColor:LinearColor=<rgba 0.25 0.5 0.4866 0.2>
    Particles.LightningGlowSoftTopDynamic1:Vector4f=&MakeVector4(
     X=&MakeFloatFromVector4(Vector 4=$Particles.DynamicMaterialParameter, Channel=::NewEnumerator0)
     Y=&Multiply_Float(A=&MakeFloatFromVector4(Vector 4=$Particles.DynamicMaterialParameter1, Channel=::NewEnumerator1), B=4)
     Z=&MakeFloatFromVector4(Vector 4=$Particles.DynamicMaterialParameter1, Channel=::NewEnumerator2)
     W=0
    )
    Particles.Scale=&VectorFromFloat(Value=&FloatFromCurve(FloatCurve=DataInterface() {propertyValues {Curve {keys [{arriveTangent=4.1785, interpMode=RCIM_Cubic, leaveTangent=1.1626, tangentMode=RCTM_Break, time=0, value=0.666}, <at 1 1>]}}}))
    Particles.SphereGlowColor:LinearColor=<rgba 0 0.25 1 1>
    Particles.SphereGlowScale:Vector3f=<1 1 1>
   ) {isSetParameters=true}
  ]
  rendererClasses [&NiagaraMeshRendererProperties, &NiagaraLightRendererProperties]
  renderers [
   {rendererClass=&NiagaraMeshRendererProperties, rendererIndex=0}
   {rendererClass=&NiagaraMeshRendererProperties, rendererIndex=1}
   {rendererClass=&NiagaraMeshRendererProperties, rendererIndex=2}
   {rendererClass=&NiagaraMeshRendererProperties, rendererIndex=3}
   {rendererClass=&NiagaraMeshRendererProperties, rendererIndex=4}
   {rendererClass=&NiagaraLightRendererProperties, rendererIndex=5}
   {rendererClass=&NiagaraMeshRendererProperties, rendererIndex=6}
  ]
 }
}
summary {
 systemName=NS_Backpack_PitGlass
 emitters [
  {
   emitterName=SplineThickenMeshes
   rendererClasses [&NiagaraMeshRendererProperties, &NiagaraLightRendererProperties]
  }
  {emitterName=Sphere, rendererClasses [&NiagaraMeshRendererProperties, &NiagaraLightRendererProperties]}
 ]
 userVariables [
  User.bIsFrontEndPreview {type=NiagaraBool}
  User.bisInVehicle {type=NiagaraBool}
 ]
}
