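# Animation Blueprint for the PitGlass backpack cosmetic that makes its glass shard bones and orb bone bob continuously.
# Each update, the EventGraph computes six sine-of-real-time offsets at varying speeds and amplitudes, along X for shards and Z for the orb.
# The AnimGraph chains additive bone-space Modify Bone nodes over the input pose, sharing five offsets across twenty shards to stagger motion.
# Movement-state flags are also copied but never affect the pose.
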AnimBlueprint /BRCosmetics/Accessories/FORT_Backpacks/Backpack_M_MED_PitGlass/Meshes/M_MED_PitGlass_Pack_AnimBP
target_skeleton=&/M_MED_PitGlass_Pack_Skeleton
graphs [
 AnimGraph {nodes [
   n1 = Get Translate Shard() -> (Translate Shard:Vector)
   n2 = "Transform (Modify) Bone - Bone: dyn_shard_01"(ComponentPose=Local To Component(LocalPose=Input Pose()), Alpha=1, Translation=@n1, Rotation="0,0,0", Scale="1,1,1") {translationMode=BMM_Additive, translationSpace=BCS_BoneSpace, alphaBoolBlend {blendOption=Linear}, boneToModify {boneName=dyn_shard_01}}
   n3 = Get Translate Shard_02() -> (Translate Shard_02:Vector)
   n4 = "Transform (Modify) Bone - Bone: dyn_shard_02"(ComponentPose=@n2, Alpha=1, Translation=@n3, Rotation="0,0,0", Scale="1,1,1") {translationMode=BMM_Additive, translationSpace=BCS_BoneSpace, alphaBoolBlend {blendOption=Linear}, boneToModify {boneName=dyn_shard_02}}
   n5 = Get Translate Shard_03() -> (Translate Shard_03:Vector)
   n6 = "Transform (Modify) Bone - Bone: dyn_shard_03"(ComponentPose=@n4, Alpha=1, Translation=@n5, Rotation="0,0,0", Scale="1,1,1") {translationMode=BMM_Additive, translationSpace=BCS_BoneSpace, alphaBoolBlend {blendOption=Linear}, boneToModify {boneName=dyn_shard_03}}
   n7 = Get Translate Shard_04() -> (Translate Shard_04:Vector)
   n8 = "Transform (Modify) Bone - Bone: dyn_shard_04"(ComponentPose=@n6, Alpha=1, Translation=@n7, Rotation="0,0,0", Scale="1,1,1") {translationMode=BMM_Additive, translationSpace=BCS_BoneSpace, alphaBoolBlend {blendOption=Linear}, boneToModify {boneName=dyn_shard_04}}
   n9 = "Transform (Modify) Bone - Bone: dyn_shard_05"(ComponentPose=@n8, Alpha=1, Translation=Get Translate Shard_05() -> (Translate Shard_05:Vector), Rotation="0,0,0", Scale="1,1,1") {translationMode=BMM_Additive, translationSpace=BCS_BoneSpace, alphaBoolBlend {blendOption=Linear}, boneToModify {boneName=dyn_shard_05}}
   n10 = "Transform (Modify) Bone - Bone: dyn_shard_06"(ComponentPose=@n9, Alpha=1, Translation=@n1, Rotation="0,0,0", Scale="1,1,1") {translationMode=BMM_Additive, translationSpace=BCS_BoneSpace, alphaBoolBlend {blendOption=Linear}, boneToModify {boneName=dyn_shard_06}}
   n11 = "Transform (Modify) Bone - Bone: dyn_shard_07"(ComponentPose=@n10, Alpha=1, Translation=@n3, Rotation="0,0,0", Scale="1,1,1") {translationMode=BMM_Additive, translationSpace=BCS_BoneSpace, alphaBoolBlend {blendOption=Linear}, boneToModify {boneName=dyn_shard_07}}
   n12 = "Transform (Modify) Bone - Bone: dyn_shard_08"(ComponentPose=@n11, Alpha=1, Translation=@n5, Rotation="0,0,0", Scale="1,1,1") {translationMode=BMM_Additive, translationSpace=BCS_BoneSpace, alphaBoolBlend {blendOption=Linear}, boneToModify {boneName=dyn_shard_08}}
   n13 = "Transform (Modify) Bone - Bone: dyn_shard_09"(ComponentPose=@n12, Alpha=1, Translation=@n7, Rotation="0,0,0", Scale="1,1,1") {translationMode=BMM_Additive, translationSpace=BCS_BoneSpace, alphaBoolBlend {blendOption=Linear}, boneToModify {boneName=dyn_shard_09}}
   n14 = "Transform (Modify) Bone - Bone: dyn_shard_10"(ComponentPose=@n13, Alpha=1, Translation=@n3, Rotation="0,0,0", Scale="1,1,1") {translationMode=BMM_Additive, translationSpace=BCS_BoneSpace, alphaBoolBlend {blendOption=Linear}, boneToModify {boneName=dyn_shard_10}}
   n15 = Get Translate Shard()
   n16 = "Transform (Modify) Bone - Bone: dyn_shard_11"(ComponentPose=@n14, Alpha=1, Translation=@n15, Rotation="0,0,0", Scale="1,1,1") {translationMode=BMM_Additive, translationSpace=BCS_BoneSpace, alphaBoolBlend {blendOption=Linear}, boneToModify {boneName=dyn_shard_11}}
   n17 = Get Translate Shard_02()
   n18 = "Transform (Modify) Bone - Bone: dyn_shard_12"(ComponentPose=@n16, Alpha=1, Translation=@n17, Rotation="0,0,0", Scale="1,1,1") {translationMode=BMM_Additive, translationSpace=BCS_BoneSpace, alphaBoolBlend {blendOption=Linear}, boneToModify {boneName=dyn_shard_12}}
   n19 = Get Translate Shard_05()
   n20 = "Transform (Modify) Bone - Bone: dyn_shard_13"(ComponentPose=@n18, Alpha=1, Translation=@n19, Rotation="0,0,0", Scale="1,1,1") {translationMode=BMM_Additive, translationSpace=BCS_BoneSpace, alphaBoolBlend {blendOption=Linear}, boneToModify {boneName=dyn_shard_13}}
   n21 = "Transform (Modify) Bone - Bone: dyn_shard_14"(ComponentPose=@n20, Alpha=1, Translation=@n15, Rotation="0,0,0", Scale="1,1,1") {translationMode=BMM_Additive, translationSpace=BCS_BoneSpace, alphaBoolBlend {blendOption=Linear}, boneToModify {boneName=dyn_shard_14}}
   n22 = Get Translate Shard_03()
   n23 = "Transform (Modify) Bone - Bone: dyn_shard_15"(ComponentPose=@n21, Alpha=1, Translation=@n22, Rotation="0,0,0", Scale="1,1,1") {translationMode=BMM_Additive, translationSpace=BCS_BoneSpace, alphaBoolBlend {blendOption=Linear}, boneToModify {boneName=dyn_shard_15}}
   n24 = "Transform (Modify) Bone - Bone: dyn_shard_16"(ComponentPose=@n23, Alpha=1, Translation=@n15, Rotation="0,0,0", Scale="1,1,1") {translationMode=BMM_Additive, translationSpace=BCS_BoneSpace, alphaBoolBlend {blendOption=Linear}, boneToModify {boneName=dyn_shard_16}}
   n25 = "Transform (Modify) Bone - Bone: dyn_shard_17"(ComponentPose=@n24, Alpha=1, Translation=@n17, Rotation="0,0,0", Scale="1,1,1") {translationMode=BMM_Additive, translationSpace=BCS_BoneSpace, alphaBoolBlend {blendOption=Linear}, boneToModify {boneName=dyn_shard_17}}
   n26 = "Transform (Modify) Bone - Bone: dyn_shard_18"(ComponentPose=@n25, Alpha=1, Translation=@n22, Rotation="0,0,0", Scale="1,1,1") {translationMode=BMM_Additive, translationSpace=BCS_BoneSpace, alphaBoolBlend {blendOption=Linear}, boneToModify {boneName=dyn_shard_18}}
   n27 = "Transform (Modify) Bone - Bone: dyn_shard_19"(ComponentPose=@n26, Alpha=1, Translation=Get Translate Shard_04(), Rotation="0,0,0", Scale="1,1,1") {translationMode=BMM_Additive, translationSpace=BCS_BoneSpace, alphaBoolBlend {blendOption=Linear}, boneToModify {boneName=dyn_shard_19}}
   n28 = "Transform (Modify) Bone - Bone: dyn_shard_20"(ComponentPose=@n27, Alpha=1, Translation=@n19, Rotation="0,0,0", Scale="1,1,1") {translationMode=BMM_Additive, translationSpace=BCS_BoneSpace, alphaBoolBlend {blendOption=Linear}, boneToModify {boneName=dyn_shard_20}}
   n29 = "Transform (Modify) Bone - Bone: dyn_orb"(ComponentPose=@n28, Alpha=0.85, Translation=Get Translate Shard_06() -> (Translate Shard_06:Vector), Rotation="0,0,0", Scale="1,1,1") {translationMode=BMM_Additive, translationSpace=BCS_BoneSpace, alphaBoolBlend {blendOption=Linear}, boneToModify {boneName=dyn_orb}}
   Output Pose(Result=Component To Local(ComponentPose=@n29)) {layerGroup=None}
  ]}
 EventGraph {nodes [
   Event BlueprintUpdateAnimation()
   > n1 = Sequence!(execute) -> (then_0, then_1, then_2:Exec, then_3:Exec, then_4:Exec, then_5:Exec, then_6:Exec, then_7:Exec, then_8:Exec, then_9:Exec)
   n2 = "Sin (Radians)"(A=float * float(A=GetRealTimeSeconds(), B=2.3))
   > Set Translate Shard(Translate Shard:Vector=MakeVector(X=float / float(A=@n2, B=-1.4) {is=title K2Node_CallFunction}, Y=0)) -> (Output_Get:Vector)
   n3 = "Sin (Radians)"(A=float * float(A=GetRealTimeSeconds(), B=1.6))
   Set Translate Shard_02(execute=@n1.then_1, Translate Shard_02:Vector=MakeVector(X=float / float(A=@n3, B=-1.1), Y=0)) -> (Output_Get:Vector)
   n4 = "Sin (Radians)"(A=float * float(A=GetRealTimeSeconds(), B=1.5))
   Set Translate Shard_03(execute=@n1.then_2, Translate Shard_03:Vector=MakeVector(X=float / float(A=@n4, B=-1.2), Y=0)) -> (Output_Get:Vector)
   n5 = "Sin (Radians)"(A=float * float(A=GetRealTimeSeconds(), B=2))
   Set Translate Shard_04(execute=@n1.then_3, Translate Shard_04:Vector=MakeVector(X=float / float(A=@n5, B=-1.2), Y=0)) -> (Output_Get:Vector)
   n6 = "Sin (Radians)"(A=float * float(A=GetRealTimeSeconds(), B=0.8))
   Set Translate Shard_05(execute=@n1.then_4, Translate Shard_05:Vector=MakeVector(X=float / float(A=@n6, B=-1.2), Y=0)) -> (Output_Get:Vector)
   n7 = "Sin (Radians)"(A=float * float(A=GetRealTimeSeconds(), B=1.6))
   Set Translate Shard_06(execute=@n1.then_5, Translate Shard_06:Vector=MakeVector(Y=0, Z=float / float(A=@n7, B=-0.8))) -> (Output_Get:Vector)
   Set Is On Ground(execute=@n1.then_6, Is On Ground:bool=Get bIsOnGround() -> (bIsOnGround:bool)) -> (Output_Get:bool)
   Set Is Moving 2D(execute=@n1.then_7, Is Moving 2D:bool=Get bIsMoving2D()) -> (Output_Get:bool)
   Set Is Sky Diving(execute=@n1.then_8, Is Sky Diving:bool=Get bIsSkyDiving()) -> (Output_Get:bool)
   Set Is Jumping(execute=@n1.then_9, Is Jumping:bool=Get bIsJumping()) -> (Output_Get:bool)
   TryGetPawnOwner()
  ]}
]
variables [
 Is Falling {type=boolean}
 Is Jumping {type=boolean}
 Is Moving 2D {type=boolean}
 Is On Ground {type=boolean}
 Is Sky Diving {type=boolean}
 Translate Shard {type=Vector, default=<0.7338 0 0>}
 Translate Shard_02 {type=Vector, default=<-0.571 0 0>}
 Translate Shard_03 {type=Vector, default=<-0.746 0 0>}
 Translate Shard_04 {type=Vector, default=<0.6592 0 0>}
 Translate Shard_05 {type=Vector, default=<0.272 0 0>}
 Translate Shard_06 {type=Vector, default=<0 0 -0.7137>}
]
